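; Interactable egg-like prop that a player incubates through a long interaction to hatch an AI creature, normally converted to that player.
; Interaction applies an incubation GameplayEffect, then requests a spawn through the AI spawner; conversion follows an optional failsafe delay.
; RepNotifies drive client incubation animation and hatching, which launches the nearby player, fires cues and swaps in a spent prop.
; Losing all health cancels incubation and destroys the prop; interaction time comes from a hotfix curve table.
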(event EventBeginPlay
  (bind _ambient_looping_sound (Variables|Default|GetAmbient%20Looping%20Sound))
  (bind _sound (Variables|/Script/Engine.AudioComponent|GetSound _ambient_looping_sound))
  (bind (_interaction_time_hotfix_table_curvetable _interaction_time_hotfix_table_rowname) (Utilities|Struct|BreakCurveTableRowHandle (Variables|GetInteraction%20Time%20Hotfix%20Table_CurveTable)))
  (bind _staticmesh (Variables|Default|GetStaticMesh))
  (Health|BindEventtoOnBuildingHealthChanged (delegate Custom|BuildingHealthChanged))
  (Variables|Prop|SetPropMeshInitialPos (Variables|/Script/Engine.SceneComponent|GetRelativeLocation _staticmesh))
  (Variables|Prop|SetPropMeshInitialScale (Transformation|GetWorldScale _staticmesh))
  ;[1 1 1] Hotfix interaction time
  (bind _outxy (DataTable|EvaluateCurveTableRow _interaction_time_hotfix_table_curvetable _interaction_time_hotfix_table_rowname)
    (:RowFound
      (Variables|Interaction|SetInteraction%20Time _outxy)
      ;[1 1 1] Start the AV indicator with a sound reference (for attenuation data)
      (Utilities|IsValid _sound
        (:Is%20Valid
          (SoundIndicator|StartTargetedSoundIndicatorLooping (Variables|Default|GetFortSoundIndicator) self _sound 1.0 0))))))

(event Custom|BuildingHealthChanged
  (if (<= (Building|GetBuildingHealthPercentage self) 0)
    (if (not (Variables|Prop|GetHasHatched))
      (CallFunction|IncubationCancel))
    (Default|DestroyProp)))

(event EventEndPlay
  (params (EndPlayReason :type EEndPlayReason%20Enum :default "Destroyed"))
  (SoundIndicator|StopSoundIndicatorLooping (Variables|Default|GetFortSoundIndicator)))

(event Interaction|EventOnServerNotifyStartLongUse
  (params (InteractingPawn :type Fort%20Player%20Pawn%20Object%20Reference))
  ;[1 1 1] Interaction flow
  (Utilities|IsValid InteractingPawn
    (:Is%20Valid
      (Networking|FlushNetDormancy)
      (Variables|SpawnAi|SetConverterPlayer InteractingPawn)
      (CallFunction|IncubationStart))))

(event Interaction|EventOnServerNotifyEndLongUse
  (params (InteractingPawn :type Fort%20Player%20Pawn%20Object%20Reference) (bUseCompleted :type Boolean))
  ;[1 1 1] Interaction flow
  (if (not bUseCompleted)
    (CallFunction|IncubationCancel)))

(event Interaction|EventBlueprintOnInteract
  (params (InteractingPawn :type Fort%20Pawn%20Object%20Reference) (InteractionBeingAttempted :type EInteractionBeingAttempted%20Enum :default "FirstInteraction"))
  ;[1 1 1] Interaction flow
  (Networking|FlushNetDormancy)
  (Variables|SpawnAi|SetConverterPlayer InteractingPawn)
  (CallFunction|IncubationComplete))

(event Custom|Incubation_Start
  ;[1 0.8918 0.2289] Incubation Flow
  (Networking|FlushNetDormancy)
  (Variables|Prop|SetIncubationStarted true)
  (CallFunction|IncubationGEApply))

(event Custom|Incubation_Cancel
  ;[1 0.8918 0.2289] Incubation Flow
  (Networking|FlushNetDormancy)
  (CallFunction|IncubationGERemove)
  (Variables|Prop|SetIncubationStarted false)
  (Variables|SpawnAi|SetConverterPlayer 0))

(event Custom|Incubation_Complete
  ;[1 0.8918 0.2289] Incubation Flow
  (Networking|FlushNetDormancy)
  (Variables|Interaction|SetCanInteract false)
  (CallFunction|IncubationGERemove)
  (CallFunction|SpawnCreature))

(event Custom|IncubationGE_Apply
  ;[0 0.08394 1] GEs
  (bind _converterplayer (Variables|SpawnAi|GetConverterPlayer)
    (:then
      (GameplayEffects|ApplyGameplayEffectToSelf :self (Ability|GetFortAbilitySystemComponent _converterplayer) :GameplayEffectClass (Variables|Prop|GetGE_PlayerPropIncubation) :Level 0.0))
    (:else)))

(event Custom|IncubationGE_Remove
  ;[0 0.08394 1] GEs
  (bind _converterplayer (Variables|SpawnAi|GetConverterPlayer)
    (:then
      (GameplayEffects|RemoveActiveGameplayEffectbySourceEffect (Ability|GetFortAbilitySystemComponent _converterplayer) (Variables|Prop|GetGE_PlayerPropIncubation) 0 -1))
    (:else)))

(event Custom|SpawnCreature
  ;[0.3747 1 0.6875] Creature spawning
  (Networking|FlushNetDormancy)
  (Variables|Prop|SetHasHatched true)
  (Utilities|FlowControl|Delay 0.1)
  (bind _returnvalue (AI|GetAISpawner))
  (EventDispatchers|BindEventtoOnPawnSpawned _returnvalue (delegate Custom|OnPawnSpawned))
  (bind _returnvalue_2 (Class|FortAthenaAISpawnerData|CreateComponentListfromClass (Variables|SpawnAi|GetAISpawner%20Data%20Class) self))
  (bind _converterplayer (Variables|SpawnAi|GetConverterPlayer)
    (:then
      (Variables|SpawnAi|SetCreatureSpawnRotation (Transformation|GetActorRotation _converterplayer)))
    (:else
      (Variables|SpawnAi|SetCreatureSpawnRotation (Transformation|GetActorRotation))))
  (bind _returnvalue_3 (Class|AthenaAISpawner|RequestSpawn _returnvalue _returnvalue_2 (Math|Transform|MakeTransform (+ (Transformation|GetActorLocation) (Variables|SpawnAi|GetSpawned%20AI%20Position%20Offset)) (Variables|SpawnAi|GetCreatureSpawnRotation)) true -1))
  (Variables|SpawnAi|SetSpawnRequestID _returnvalue_3))

(event Custom|OnPawnSpawned
  (params (Pawn :type Pawn%20Object%20Reference) (RequestID :type Integer))
  (bind _convertfailsafedelay (Variables|SpawnAi|GetConvertFailsafeDelay))
  ;[0.3747 1 0.6875] Creature spawning
  (if (== RequestID (Variables|SpawnAi|GetSpawnRequestID))
    (CallFunction|SendHatchedQuestStatEvent Pawn)
    (Class|FortAthenaLivingWorldManager|RegisterDensityReservoir (Class|FortAthenaLivingWorldManager|GetLivingWorldManager) Pawn 1.0 -1.0)
    (if (> _convertfailsafedelay 0)
      (Utilities|FlowControl|Delay _convertfailsafedelay)
      (else
        _))
    (bind _asfort_pawn (Utilities|Casting|CastToFortPawn Pawn))
    (if (not (GameplayTags|HasMatchingGameplayTag _asfort_pawn "(TagName=\"Athena.Gameplay.Status.Converted\")"))
      (bind _converterplayer (Variables|SpawnAi|GetConverterPlayer)
        (:then
          (if (not (GameplayTags|HasMatchingGameplayTag _converterplayer "(TagName=\"Granted.Riding.Player.IsRiding\")"))
            (Convert|TriggerConversion _converterplayer _asfort_pawn (Variables|SpawnAi|GetCan%20Steal%20Spawned%20Tamed%20AI))
            (bind _asfort_aipawn (Utilities|Casting|CastToFortAIPawn _asfort_pawn))
            (AI|SetCanInteract _asfort_aipawn true)))
        (:else)))))

(event Custom|IncubationPropMovement
  ;[0.3583 0.3382 1] Egg animation
  (Components|Timeline|SetPlayRate (Variables|BGA_PropAISpawner|GetPropIncubatingTimeline) (/ 1.0 (Variables|Interaction|GetInteraction%20Time)))
  (bind (_direction _scalecurve _positioncurve _eggphase _curvetime) (Timeline|PropIncubatingTimeline PlayFromStart)
    (:Update
      (Rendering|Material|SetCustomPrimitiveDataFloat (Variables|Default|GetStaticMesh) 18 _curvetime))
    (:IncubationPulseEvent
      (switch Utilities|FlowControl|Switch|SwitchonETimelineDirection _direction
        (:Forward
          (GameplayCue|ExecuteGameplayCue%28Burst%29Local self (Variables|Prop|GetIncubatePulseGameplayCueTag) (Ability|GameplayCue|MakeGameplayCueParameters :NormalizedMagnitude _eggphase :RawMagnitude 0.0 :Location "0,0,0" :Normal "0,0,0" :SourceObject (Variables|Default|GetStaticMesh) :GameplayEffectLevel 1 :AbilityLevel 1 :bReplicateLocationWhenUsingMinimalRepProxy false)))))))

(event Custom|IncubationCancelledPropMovement
  ;[0.3583 0.3382 1] Egg animation
  (Components|Timeline|SetPlayRate (Variables|BGA_PropAISpawner|GetPropIncubatingTimeline) (* (/ 1.0 (Variables|Interaction|GetInteraction%20Time)) 4.0))
  (Timeline|PropIncubatingTimeline Reverse))

(event Custom|DestroyActorAfterDelay
  (params (Delay :type Float%20%28double-precision%29))
  (Utilities|FlowControl|Delay Delay)
  (Actor|DestroyActor))

(fn BlueprintCanInteract
  (params (InteractingPawn :type Fort%20Pawn%20Object%20Reference) (InteractionBeingAttempted :type EInteractionBeingAttempted%20Enum :default "FirstInteraction") (InteractionType :type TInteractionType%20Enum :default "IT_NoInteraction"))
  (bind _converterplayer (Variables|SpawnAi|GetConverterPlayer)
    (:then
      (if (== _converterplayer InteractingPawn)
        _))
    (:else
      _))
  (return (Math|Boolean|ANDBoolean (not (GameplayTags|HasAnyMatchingGameplayTags InteractingPawn (Variables|Interaction|GetBlockPropIncubationTagContainer))) (Variables|Interaction|GetCanInteract) (<= (Math|Vector|Distance(Vector) (Transformation|GetActorLocation) (Transformation|GetActorLocation InteractingPawn)) (Variables|Interaction|GetInteractMaxRange)))))

(fn BlueprintGetInteractionString
  (params (InteractingPawn :type Fort%20Pawn%20Object%20Reference) (InteractionBeingAttempted :type EInteractionBeingAttempted%20Enum :default "FirstInteraction"))
  (return (Variables|Interaction|GetInteractionString)))

(fn BlueprintGetInteractionTime
  (params (InteractingPawn :type Fort%20Pawn%20Object%20Reference) (InteractionBeingAttempted :type EInteractionBeingAttempted%20Enum :default "FirstInteraction"))
  (return (Variables|Interaction|GetInteraction%20Time) true))

(fn DestroyProp
  (bind _staticmesh (Variables|Default|GetStaticMesh))
  (Collision|SetCollisionEnabled _staticmesh "NoCollision")
  (Variables|/Script/Engine.PrimitiveComponent|SetCanCharacterStepUpOn "ECB_No" _staticmesh)
  (CallFunction|DestroyActorAfterDelay 5.0)
  (CallFunction|FadeOutAmbientSound 0.0)
  (Networking|SwitchHasAuthority
    (:Remote
      (GameplayCue|ExecuteGameplayCue%28Burst%29Local self (Variables|Prop|GetDestroyGameplayCueTag) (Ability|GameplayCue|MakeGameplayCueParameters :NormalizedMagnitude 0.0 :RawMagnitude 0.0 :Location (Transformation|GetWorldLocation (Variables|Default|GetHatchFxSpawnPoint)) :Normal "0,0,0" :GameplayEffectLevel 1 :AbilityLevel 1 :bReplicateLocationWhenUsingMinimalRepProxy false))
      (Rendering|SetVisibility (Variables|Default|GetStaticMesh) false false))))

(fn FadeOutAmbientSound
  (params (FadeOutDuration :type Float%20%28double-precision%29 :default 0.25))
  (Audio|Components|Audio|FadeOut (Variables|Default|GetAmbient%20Looping%20Sound) FadeOutDuration 0.0 "Linear" :!class /Script/Engine.AudioComponent)
  (SoundIndicator|StopSoundIndicatorLooping (Variables|Default|GetFortSoundIndicator)))

(fn GetCurieCollisionComponentOverride
  (return (Variables|Default|GetStaticMesh)))

(fn GetPlayerIncubatePosition
  (params (PlayerActor :type Actor%20Object%20Reference))
  (return (Utilities|Operators|Add (Transformation|GetActorLocation) (* (Math|Vector|GetUnitDirection(Vector) (Math|Conversions|ToVector(Vector2D) (Math|Conversions|ToVector2D(Vector) (Transformation|GetActorLocation))) (Math|Conversions|ToVector(Vector2D) (Math|Conversions|ToVector2D(Vector) (Transformation|GetActorLocation PlayerActor)))) (.x (Variables|Prop|GetIncubationPlayerPosOffset))) (* (.y (Variables|Prop|GetIncubationPlayerPosOffset)) (Math|Vector|CrossProduct (Math|Vector|GetUnitDirection(Vector) (Math|Conversions|ToVector(Vector2D) (Math|Conversions|ToVector2D(Vector) (Transformation|GetActorLocation))) (Math|Conversions|ToVector(Vector2D) (Math|Conversions|ToVector2D(Vector) (Transformation|GetActorLocation PlayerActor)))) (Math|Vector|VectorUp))) (* (.z (Variables|Prop|GetIncubationPlayerPosOffset)) (Math|Vector|VectorUp)))))

(fn HatchProp
  (bind (_returnvalue_location _returnvalue_rotation _returnvalue_scale) (Math|Transform|BreakTransform (Transformation|GetActorTransform)))
  (bind _spent_prop_spawner (Variables|Prop|GetSpent%20Prop%20Spawner))
  (bind _staticmesh (Variables|Default|GetStaticMesh))
  (bind _converterplayer (Variables|SpawnAi|GetConverterPlayer)
    (:then
      (if (<= (Math|Vector|Distance(Vector) (Transformation|GetActorLocation _converterplayer) (Transformation|GetActorLocation)) (Variables|Prop|GetDistFromPropHatchingToGetLaunchedUp))
        (Character|LaunchCharacter _converterplayer (+ (* (Math|Vector|VectorUp) (Math|Conversions|ToVector(Float) (Variables|Prop|GetHatch_PlayerVerticalLaunchVelocity))) (* (Math|Vector|GetUnitDirection(Vector) (Transformation|GetActorLocation _converterplayer) (Transformation|GetActorLocation)) (Math|Float|MapRangeClamped (Math|Vector|Distance(Vector) (Transformation|GetActorLocation _converterplayer) (Transformation|GetActorLocation)) 0.0 (Variables|Interaction|GetInteractMaxRange) 0.0 (Variables|Prop|GetHatch_PlayerHorizontalMaxLaunchVelocity)))) true true)
        (else
          _))
      (Collision|SetCollisionEnabled _staticmesh "NoCollision")
      (Variables|/Script/Engine.PrimitiveComponent|SetCanCharacterStepUpOn "ECB_No" _staticmesh)
      (Networking|SwitchHasAuthority
        (:Authority
          (if (Utilities|IsValidClass _spent_prop_spawner)
            (bind _returnvalue (Game|SpawnActorfromClass _spent_prop_spawner (Math|Transform|MakeTransform (+ _returnvalue_location (Variables|Prop|GetSpent%20Prop%20Spawner%20Offset)) _returnvalue_rotation _returnvalue_scale) "Undefined" "MultiplyWithRoot" 0 0 0 0))
            (Utilities|IsValid _returnvalue
              (:Is%20Valid
                (Class|PropSpentAISpawner|BindtoLivingWorldPrefabDestruction _returnvalue (Variables|Default|GetOwningLivingWorldPrefabRef)))
              (:Is%20Not%20Valid
                (Class|BuildingFunctionLibrary|LogString "NewEnumerator4" (Utilities|String|Append "BGA_PropAISpawner couldn't spawn its spent version. \\n PropAISpawner: " (Utilities|GetDisplayName self) "\\n PropSpentAISpawner: " (Utilities|GetClassDisplayName _spent_prop_spawner)) true self)))
            (else
              _))
          (Actor|DestroyActor)
          (CallFunction|FadeOutAmbientSound 0.25))
        (:Remote
          (GameplayCue|ExecuteGameplayCue%28Burst%29Local self (Variables|Prop|GetHatchGameplayCueTag) (Ability|GameplayCue|MakeGameplayCueParameters :NormalizedMagnitude 0.0 :RawMagnitude 0.0 :Location (Transformation|GetWorldLocation (Variables|Default|GetHatchFxSpawnPoint)) :Normal "0,0,0" :GameplayEffectLevel 1 :AbilityLevel 1 :bReplicateLocationWhenUsingMinimalRepProxy false))
          (bind _converterplayer_2 (Variables|SpawnAi|GetConverterPlayer)
            (:then
              (GameplayCue|ExecuteGameplayCue%28Burst%29Local _converterplayer_2 (Variables|Prop|GetPlayer_HatchProp_GameplayCueTag) (Ability|GameplayCue|MakeGameplayCueParameters 0.0)))
            (:else
              _))
          (Rendering|SetVisibility (Variables|Default|GetStaticMesh) false false))))
    (:else
      _)))

(fn OnRep_CanInteract)

(fn OnRep_ConverterPlayer)

(fn OnRep_HasHatched
  (Default|HatchProp)
  (CallFunction|FadeOutAmbientSound 0.25))

(fn OnRep_IncubationStarted
  (Networking|SwitchHasAuthority
    (:Remote
      (if (Variables|Prop|GetIncubationStarted)
        (CallFunction|IncubationPropMovement)
        (else
          (CallFunction|IncubationCancelledPropMovement))))))

(fn SendHatchedQuestStatEvent
  (params (Spawned%20Pawn :type Pawn%20Object%20Reference))
  (bind _converterplayer (Variables|SpawnAi|GetConverterPlayer))
  (bind _returnvalue (Fortnite|GetFortPlayerControllerfromActor _converterplayer))
  (bind _returnvalue_2 (Quests|GetQuestManager _returnvalue "Invalid"))
  (Utilities|IsValid _converterplayer
    (:Is%20Valid
      (Utilities|IsValid _returnvalue
        (:Is%20Valid
          (Utilities|IsValid _returnvalue_2
            (:Is%20Valid
              (Utilities|IsValid Spawned%20Pawn
                (:Is%20Valid
                  (bind _returnvalue_3 (GameplayTags|GetOwnedGameplayTags Spawned%20Pawn :!class /Script/GameplayTags.GameplayTagAssetInterface))
                  (Quest|SendCustomStatEventwithTags :self _returnvalue_2 :Type "Deprecated" :TargetTags _returnvalue_3 :Count 1))))))))))

(fn ConstructionScript)
