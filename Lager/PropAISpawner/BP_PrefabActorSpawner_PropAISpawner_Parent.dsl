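; Parent prefab actor spawner for the Lager prop AI spawner, placing spawned actors at its attached nest actor.
; On BeginPlay it hides its editor-only static mesh and binds a handler to the actor-spawned callback.
; When a nest actor is available, the handler moves the spawned actor to it with slight random up-axis jitter and random local yaw.
; It then casts the actor to BGA_PropAISpawner and sets its owning Living World prefab reference.
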(event EventBeginPlay
  (Rendering|SetVisibility (Variables|Default|GetStaticMeshEditorOnly) false false)
  (CallFunction|CallwhenActorSpawned (delegate Custom|ActorSpawned)))

(event Collision|EventActorBeginOverlap
  (params (OtherActor :type Actor%20Object%20Reference)))

(event Custom|ActorSpawned
  (params (SpawnedActor :type Actor%20Object%20Reference))
  (bind _attachednestactor (Variables|Default|GetAttachedNestActor)
    (:then
      (Transformation|SetActorTransform SpawnedActor (Math|Transform|MakeTransform (+ (.location (Transformation|GetActorTransform _attachednestactor)) (* (Transformation|GetActorUpVector _attachednestactor) (Math|Random|RandomFloatinRange -4.0 4.0))) (.rotation (Transformation|GetActorTransform _attachednestactor))) false true)
      (Transformation|AddActorLocalRotation SpawnedActor (Math|Rotator|MakeRotator 0.0 0.0 (Math|Random|RandomFloatinRange 0.0 359.0)) false true))
    (:else
      _))
  (bind _asbga_prop_aispawner (Utilities|Casting|CastToBGA_PropAISpawner SpawnedActor))
  (Variables|/Lager/PropAISpawner/BGA_PropAISpawner.BGA_PropAISpawner_C|SetOwningLivingWorldPrefabRef (Class|FortAthenaLivingWorldPrefab|GetOwningLivingWorldPrefab true) _asbga_prop_aispawner))

(event EventTick
  (params (DeltaSeconds :type Float%20%28single-precision%29)))

(fn ConstructionScript)
