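; Fortnite prop actor for a spent AI spawner, tied to the lifetime of a Living World prefab.
; BindToLivingWorldPrefabDestruction takes the owning prefab and, if it is valid, binds its OnDestroyed event to Destroy, which removes this prop.
; On death, the prop spawns its configured death FX at the static mesh component's world location.
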(event Custom|BindToLivingWorldPrefabDestruction
  (params (OwningLivingWorldPrefab :type Fort%20Athena%20Living%20World%20Prefab%20Object%20Reference))
  (Utilities|IsValid OwningLivingWorldPrefab
    (:Is%20Valid
      (Game|BindEventtoOnDestroyed OwningLivingWorldPrefab (delegate Custom|Destroy)))))

(event Custom|Destroy
  (params (DestroyedActor :type Actor%20Object%20Reference))
  (Actor|DestroyActor))

(event EventOnDeathPlayEffects
  (params (Damage :type Float%20%28single-precision%29) (DamageTags :type Gameplay%20Tag%20Container%20Structure%20%28by%20ref%29) (Momentum :type Vector :default "0, 0, 0") (HitInfo :type Hit%20Result%20Structure%20%28by%20ref%29) (InstigatedBy :type Fort%20Pawn%20Object%20Reference) (DamageCauser :type Actor%20Object%20Reference) (EffectContext :type Gameplay%20Effect%20Context%20Handle%20Structure))
  (Effects|Components|ParticleSystem|SpawnFXAtLocation (Variables|Default|GetFX%20on%20Prop%20Death) (Transformation|GetWorldLocation (Variables|Debug|GetStaticMeshComponent))))

(fn ConstructionScript)
